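# Customizable Fortnite landscape master material with paintable layers, optional cliffs on steep areas and a landscape grass output.
# Layers sample tintable world-aligned textures on the Z projection, with quality switches substituting scalar surface values and vertex normals on mobile.
# Height-contrast functions blend layers in sequence, using layer switches to skip unpainted ones and carrying height in Opacity Mask.
# Layer7 gets a height-lerped mask for adjusting underlying layers, such as wetness or scorch marks.
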Material /FortniteLandscape/Materials/M_FortniteLandscape_Customizable
graph {
 properties {
  bTangentSpaceNormal=false
  bUseMaterialAttributes=true
  bUsedWithNanite=true
  pixelDepthOffsetMode=PDOM_Legacy
  usedShadingModels=Default Lit
 }
 nodes [
  Constant() {r=2500}
  Constant() {r=2}
  Constant() {r=-0.5}
  Constant() {r=-1}
  Constant() {r=3}
  Constant() {desc=13000.0, r=4100}
  Constant() {desc=13000.0, r=10100}
  Constant() {r=1.5}
  Constant() {r=-0.5}
  MaterialExpressionConstant_1 = Constant()
  LandscapeGrassOutput!(Base_GrassType=@MaterialExpressionConstant_1) {grassTypes [Base_GrassType {
     grassType=/Game/Athena/Apollo/Environments/Landscape/SmallFoliage/DataAssets/LandscapeGrassType/NonRVT/LGT_StandardGrass_NoRVT
     input {
      expression=&self:MaterialExpressionConstant_1
      inputName=None
      mask=0
      maskA=0
      maskB=0
      maskG=0
      maskR=0
      outputIndex=0
     }
    }]}
  LandscapeLayerSample() {parameterName=Layer4}
  LandscapeLayerSample() {parameterName=Layer5}
  LandscapeLayerSample() {parameterName=Layer6}
  LandscapeLayerSample() {parameterName=Layer7}
  LandscapeLayerSample() {parameterName=Layer1}
  LandscapeLayerSample() {parameterName=Layer2}
  LandscapeLayerSample() {parameterName=Layer3}
  LandscapeVisibilityMask()
  MaterialFunctionCall!(
   In=SetMaterialAttributes!(MaterialAttributes=NamedRerouteUsage(), Opacity Mask=LandscapeVisibilityMask()) {attributeSetTypes [#1]}
   IsForLandscapeMaterial=StaticBool() {value=true}
  ) {materialFunction=/Game/Creative/Materials/MaterialFunctions/MF_CreativeFortniteGameplay}
  n1 = Multiply(A=Multiply(A=NamedRerouteUsage(), B=NamedRerouteUsage()), B=Multiply(A=NamedRerouteUsage(), B=NamedRerouteUsage()))
  n2 = NamedRerouteUsage()
  n3 = ScalarParameter() {
   desc="Spawn Base_GrassType meshes on Layer7.\r\nA value of one will spawn meshes where this layer is used."
   group=LandscapeGrassTypes
   parameterName=Base_GrassType_Layer7
   sortPriority=0
  }
  n4 = Multiply(
   A=Multiply(A=NamedRerouteUsage(), B=NamedRerouteUsage())
   B=Multiply(A=LinearInterpolate(A=@n2, Alpha=@n3) {constA=1}, B=NamedRerouteUsage())
  )
  n5 = ScalarParameter() {desc=Spawns Landscape Grass Type meshes on the Base Layer., group=LandscapeGrassTypes, parameterName=Base_UseGrassType, sortPriority=0}
  Multiply(A=Multiply(A=Multiply(A=@n1, B=@n4), B=LandscapeVisibilityMask()), B=@n5)
  n6 = VectorParameter() {
   desc="Divide the BaseColor_Texture by this color to adjust the color values.\r\nA vaule of 1,1,1 will not alter the colors.\r\nAlpha value is not used."
   group=01_Layer1
   parameterName=Layer1_BaseColor_Tint
   sortPriority=110
   channelNames {a=Not Used, b=Blue Value, g=Green Value, r=Red Value}
   defaultValue=<rgba 1 1 1 1>
  }
  n7 = TextureObjectParameter() {
   desc="BaseColor texture using Default compression.\r\nUse a heightmap in the alpha channel for more accurate layer blending results."
   group=01_Layer1
   parameterName=Layer1_BaseColor_Texture
   sortPriority=100
   texture=/EpicBaseTextures/Blanks/T_Default_BaseColor_White
  }
  n8 = TextureObjectParameter() {
   desc="Channel-packed texture using Masks compression.\r\nOptimal default setup: Specular uses Red channel, Roughness uses Green channel, Metallic uses Blue channel.\r\nMany of the older Fortnite textures use green channel for metallic and blue channel for roughness.\r\nLeave the default setting if not used."
   group=01_Layer1
   parameterName=Layer1_SpecularRoughnessMetallic_Texture
   samplerType=SAMPLERTYPE_Masks
   sortPriority=200
   texture=/EpicBaseTextures/Blanks/T_Default_Masks_SpecRoughMetal
  }
  n9 = ScalarParameter() {
   defaultValue=900
   desc=Scale of the textures in the level. A lower value decreases the size.
   group=01_Layer1
   parameterName=Layer1_TextureScale
   sortPriority=90
  }
  n10 = TextureObjectParameter() {
   desc="Tangent-space normal map texture using Normalmap compression.\r\nLeave the default setting if not used."
   group=01_Layer1
   parameterName=Layer1_NormalMap_Texture
   samplerType=SAMPLERTYPE_Normal
   sortPriority=300
   texture=/EpicBaseTextures/Blanks/T_Default_NormalMap
  }
  n11 = MaterialFunctionCall!(
   Diffuse Texture Object=@n7
   Spec Texture Object=@n8
   TextureSize=@n9
   --------------------------
   Use seperate settings for Z projection
   Use seperate texture for Z projection
   Texture Size Z projection
   Z Plane Diffuse Texture Object
   Z Plane Spec Texture Object
   "--------------------------      "
   ProjectionTransitionContrast
   Use Texture Alpha Heightmap to Refine the Transistion
   HeightLerpTransitionContrast
   PickTheDiffuseHeightChannel
   TheDiffuseHeightChannel
   "--------------------------       "
   NormalTextureObject=@n10
   World Space Normals
   "--------------------------  "
   Export Float 4 Diffuse=StaticBool() {value=true}
   Export Float 4 Spec
   "--------------------------     "
   WorldPosition
   World Space Normal
  ) {materialFunction=/Engine/Functions/Engine_MaterialFunctions02/Texturing/WorldAlignedTextures_Complex}
  n12 = Saturate(None=Multiply(A=ComponentMask(None=@n6) {b=true, g=true, r=true}, B=ComponentMask(None=@n11.Z Diffuse Texture) {b=true, g=true, r=true}))
  n13 = ScalarParameter() {
   defaultValue=1
   desc="Controls the intensity of the normal map texture contribution.\r\nZero equals no normal map texture contribution."
   group=01_Layer1
   parameterName=Layer1_NormalMap_Intensity
   sortPriority=310
  }
  n14 = StaticComponentMaskParameter(None=@n11.Z Spec) {
   defaultR=true
   desc="Define the SpecularRoughnessMetallic_Texture channel used for the Specular mask.\r\nDefault is red.\r\nNot used on mobile renderer."
   group=01_Layer1
   parameterName=Layer1_SpecularChannel
   sortPriority=210
  }
  n15 = StaticComponentMaskParameter(None=@n11.Z Spec) {
   defaultB=true
   desc="Define the SpecularRoughnessMetallic_Texture channel used for the Roughness mask.\r\nDefault is green.\r\nMany of the older Fortnite textures use green channel for metallic and blue channel for roughness.\r\nNot used on mobile renderer."
   group=01_Layer1
   parameterName=Layer1_RoughnessChannel
   sortPriority=220
  }
  n16 = StaticComponentMaskParameter(None=@n11.Z Spec) {
   defaultG=true
   desc="Define the SpecularRoughnessMetallic_Texture channel used for the Metallic mask.\r\nDefault is blue.\r\nMany of the older Fortnite textures use green channel for metallic and blue channel for roughness.\r\nNot used on mobile renderer."
   group=01_Layer1
   parameterName=Layer1_MetallicChannel
   sortPriority=230
  }
  n17 = VectorParameter() {
   desc="Designate which Basecolor texture channel to use for the heightmap. The heightmap will affect how this layer blends with other layers.\r\nSet all channels to zero, then set the channel you want to use to one.\r\nYou may get interesting results by playing with all these values.\r\nIf the texture doesn't have an alpha channel, leave the alpha value at zero."
   group=01_Layer1
   parameterName=Layer1_BaseColor_Heightmap_Channel
   sortPriority=120
   defaultValue=<rgba 0 0 0 1>
  }
  n18 = AppendVector(A=@n17.RGB, B=@n17.RGB)
  n19 = ScalarParameter() {
   defaultValue=1
   desc="Zero equals no heightmap contribution.\r\nThis value is clamped 0-1."
   group=01_Layer1
   parameterName=Layer1_BaseColor_Heightmap_Contribution
   sortPriority=130
  }
  n20 = LinearInterpolate(B=Saturate(None=DotProduct(A=@n11.Z Diffuse Texture, B=@n18)), Alpha=Saturate(None=@n19)) {constA=0.5}
  n21 = SetMaterialAttributes!(
   MaterialAttributes
   Base Color=@n12
   Normal=LinearInterpolate(A=VertexNormalWS(), B=@n11.Normal Z Texture, Alpha=Abs(None=@n13))
   Specular=@n14
   Roughness=@n15
   Metallic=@n16
   Opacity Mask=@n20
  ) {attributeSetTypes [
    #2
    #3
    #4
    #5
    #6
    #1
   ]}
  n22 = ScalarParameter() {
   defaultValue=0.5
   desc="Mobile renderer doesn't use a texture, only this scalar value."
   group=01_Layer1
   parameterName=Layer1_Mobile_SpecularValue
   sortPriority=240
  }
  n23 = ScalarParameter() {
   defaultValue=0.85
   desc="Mobile renderer doesn't use a texture, only this scalar value."
   group=01_Layer1
   parameterName=Layer1_Mobile_RoughnessValue
   sortPriority=250
  }
  n24 = ScalarParameter() {
   desc="Mobile renderer doesn't use a texture, only this scalar value."
   group=01_Layer1
   parameterName=Layer1_Mobile_MetallicValue
   sortPriority=260
  }
  n25 = SetMaterialAttributes(Base Color=@n12, Normal=VertexNormalWS(), Specular=@n22, Roughness=@n23, Metallic=@n24, Opacity Mask=@n20) {attributeSetTypes [
    #2
    #3
    #4
    #5
    #6
    #1
   ]}
  NamedRerouteDeclaration(None=MaterialFunctionCall!(
    Low _ PC Low and Med=@n21
    "High _ XB1, PS4"=@n21
    "Epic _ PC High, XSX, PS5"=@n21
    --------
    "Mobile Low _ PC Perf, Android"=@n25
    Mobile Med _ Switch=@n25
   ) {materialFunction=/Game/Global/MaterialFunctions/TechArt/MF_Fortnite_QualitySwitch_MaterialAttributes}.Result) {name=Layer1_Material, nodeColor=<rgba 0.5 0 1 1>}
  n26 = VectorParameter() {
   desc="Divide the BaseColor_Texture by this color to adjust the color values.\r\nA vaule of 1,1,1 will not alter the colors.\r\nAlpha value is not used."
   group=02_Layer2
   parameterName=Layer2_BaseColor_Tint
   sortPriority=110
   channelNames {a=Not Used, b=Blue Value, g=Green Value, r=Red Value}
   defaultValue=<rgba 1 1 1 1>
  }
  n27 = TextureObjectParameter() {
   desc="BaseColor texture using Default compression.\r\nUse a heightmap in the alpha channel for more accurate layer blending results."
   group=02_Layer2
   parameterName=Layer2_BaseColor_Texture
   sortPriority=100
   texture=&T_Default_BaseColor_White
  }
  n28 = TextureObjectParameter() {
   desc="Channel-packed texture using Masks compression.\r\nOptimal default setup: Specular uses Red channel, Roughness uses Green channel, Metallic uses Blue channel.\r\nMany of the older Fortnite textures use green channel for metallic and blue channel for roughness.\r\nLeave the default setting if not used."
   group=02_Layer2
   parameterName=Layer2_SpecularRoughnessMetallic_Texture
   samplerType=SAMPLERTYPE_Masks
   sortPriority=200
   texture=&T_Default_Masks_SpecRoughMetal
  }
  n29 = ScalarParameter() {
   defaultValue=900
   desc=Scale of the textures in the level. A lower value decreases the size.
   group=02_Layer2
   parameterName=Layer2_TextureScale
   sortPriority=90
  }
  n30 = TextureObjectParameter() {
   desc="Tangent-space normal map texture using Normalmap compression.\r\nLeave the default setting if not used."
   group=02_Layer2
   parameterName=Layer2_NormalMap_Texture
   samplerType=SAMPLERTYPE_Normal
   sortPriority=300
   texture=&T_Default_NormalMap
  }
  n31 = MaterialFunctionCall(
   Diffuse Texture Object=@n27
   Spec Texture Object=@n28
   TextureSize=@n29
   NormalTextureObject=@n30
   Export Float 4 Diffuse=StaticBool() {value=true}
  ) {materialFunction=&WorldAlignedTextures_Complex}
  n32 = Saturate(None=Multiply(A=ComponentMask(None=@n26) {b=true, g=true, r=true}, B=ComponentMask(None=@n31.Z Diffuse Texture) {b=true, g=true, r=true}))
  n33 = ScalarParameter() {
   defaultValue=1
   desc="Controls the intensity of the normal map texture contribution.\r\nZero equals no normal map texture contribution."
   group=02_Layer2
   parameterName=Layer2_NormalMap_Intensity
   sortPriority=310
  }
  n34 = StaticComponentMaskParameter(None=@n31.Z Spec) {
   defaultR=true
   desc="Define the SpecularRoughnessMetallic_Texture channel used for the Specular mask.\r\nDefault is red.\r\nNot used on mobile renderer."
   group=02_Layer2
   parameterName=Layer2_SpecularChannel
   sortPriority=210
  }
  n35 = StaticComponentMaskParameter(None=@n31.Z Spec) {
   defaultB=true
   desc="Define the SpecularRoughnessMetallic_Texture channel used for the Specular mask.\r\nDefault is red.\r\nNot used on mobile renderer."
   group=02_Layer2
   parameterName=Layer2_RoughnessChannel
   sortPriority=220
  }
  n36 = StaticComponentMaskParameter(None=@n31.Z Spec) {
   defaultG=true
   desc="Define the SpecularRoughnessMetallic_Texture channel used for the Specular mask.\r\nDefault is red.\r\nNot used on mobile renderer."
   group=02_Layer2
   parameterName=Layer2_MetallicChannel
   sortPriority=230
  }
  n37 = VectorParameter() {
   desc="Designate which Basecolor texture channel to use for the heightmap. The heightmap will affect how this layer blends with other layers.\r\nSet all channels to zero, then set the channel you want to use to one.\r\nYou may get interesting results by playing with all these values.\r\nIf the texture doesn't have an alpha channel, leave the alpha value at zero."
   group=02_Layer2
   parameterName=Layer2_BaseColor_Heightmap_Channel
   sortPriority=120
   defaultValue=<rgba 0 0 0 1>
  }
  n38 = AppendVector(A=@n37.RGB, B=@n37.RGB)
  n39 = ScalarParameter() {
   defaultValue=1
   desc="Zero equals no heightmap contribution.\r\nThis value is clamped 0-1."
   group=02_Layer2
   parameterName=Layer2_BaseColor_Heightmap_Contribution
   sortPriority=130
  }
  n40 = LinearInterpolate(B=Saturate(None=DotProduct(A=@n31.Z Diffuse Texture, B=@n38)), Alpha=Saturate(None=@n39)) {constA=0.5}
  n41 = SetMaterialAttributes(
   Base Color=@n32
   Normal=LinearInterpolate(A=VertexNormalWS(), B=@n31.Normal Z Texture, Alpha=Abs(None=@n33))
   Specular=@n34
   Roughness=@n35
   Metallic=@n36
   Opacity Mask=@n40
  ) {attributeSetTypes [
    #2
    #3
    #4
    #5
    #6
    #1
   ]}
  n42 = ScalarParameter() {
   defaultValue=0.5
   desc="Mobile renderer doesn't use a texture, only this scalar value."
   group=02_Layer2
   parameterName=Layer2_Mobile_SpecularValue
   sortPriority=240
  }
  n43 = ScalarParameter() {
   defaultValue=0.85
   desc="Mobile renderer doesn't use a texture, only this scalar value."
   group=02_Layer2
   parameterName=Layer2_Mobile_RoughnessValue
   sortPriority=250
  }
  n44 = ScalarParameter() {
   desc="Mobile renderer doesn't use a texture, only this scalar value."
   group=02_Layer2
   parameterName=Layer2_Mobile_MetallicValue
   sortPriority=260
  }
  n45 = SetMaterialAttributes(Base Color=@n32, Normal=VertexNormalWS(), Specular=@n42, Roughness=@n43, Metallic=@n44, Opacity Mask=@n40) {attributeSetTypes [
    #2
    #3
    #4
    #5
    #6
    #1
   ]}
  NamedRerouteDeclaration(None=MaterialFunctionCall(
    Low _ PC Low and Med=@n41
    "High _ XB1, PS4"=@n41
    "Epic _ PC High, XSX, PS5"=@n41
    "Mobile Low _ PC Perf, Android"=@n45
    Mobile Med _ Switch=@n45
   ) {materialFunction=&MF_Fortnite_QualitySwitch_MaterialAttributes}.Result) {name=Layer2_Material, nodeColor=<rgba 0.5 0 1 1>}
  n46 = ScalarParameter() {
   defaultValue=1
   desc=A higher value will sharpen the blending to other layers.
   group=Cliff
   parameterName=Cliff_BaseColor_Heightmap_LayerBlendContrast
   sortPriority=140
  }
  NamedRerouteDeclaration(None=Abs(None=@n46)) {name=Cliff_HeightBlendContrast, nodeColor=<rgba 0.5 0 1 1>}
  n47 = VectorParameter() {
   desc="Divide the BaseColor_Texture by this color to adjust the color values.\r\nA vaule of 1,1,1 will not alter the colors.\r\nAlpha value is not used."
   group=04_Layer4
   parameterName=Layer4_BaseColor_Tint
   sortPriority=110
   channelNames {a=Not Used, b=Blue Value, g=Green Value, r=Red Value}
   defaultValue=<rgba 1 1 1 1>
  }
  n48 = TextureObjectParameter() {
   desc="BaseColor texture using Default compression.\r\nUse a heightmap in the alpha channel for more accurate layer blending results."
   group=04_Layer4
   parameterName=Layer4_BaseColor_Texture
   sortPriority=100
   texture=&T_Default_BaseColor_White
  }
  n49 = TextureObjectParameter() {
   desc="Channel-packed texture using Masks compression.\r\nOptimal default setup: Specular uses Red channel, Roughness uses Green channel, Metallic uses Blue channel.\r\nMany of the older Fortnite textures use green channel for metallic and blue channel for roughness.\r\nLeave the default setting if not used."
   group=04_Layer4
   parameterName=Layer4_SpecularRoughnessMetallic_Texture
   samplerType=SAMPLERTYPE_Masks
   sortPriority=200
   texture=&T_Default_Masks_SpecRoughMetal
  }
  n50 = ScalarParameter() {
   defaultValue=900
   desc=Scale of the textures in the level. A lower value decreases the size.
   group=04_Layer4
   parameterName=Layer4_TextureScale
   sortPriority=90
  }
  n51 = TextureObjectParameter() {
   desc="Tangent-space normal map texture using Normalmap compression.\r\nLeave the default setting if not used."
   group=04_Layer4
   parameterName=Layer4_NormalMap_Texture
   samplerType=SAMPLERTYPE_Normal
   sortPriority=300
   texture=&T_Default_NormalMap
  }
  n52 = MaterialFunctionCall(
   Diffuse Texture Object=@n48
   Spec Texture Object=@n49
   TextureSize=@n50
   NormalTextureObject=@n51
   Export Float 4 Diffuse=StaticBool() {value=true}
  ) {materialFunction=&WorldAlignedTextures_Complex}
  n53 = Saturate(None=Multiply(A=ComponentMask(None=@n47) {b=true, g=true, r=true}, B=ComponentMask(None=@n52.Z Diffuse Texture) {b=true, g=true, r=true}))
  n54 = ScalarParameter() {
   defaultValue=1
   desc="Controls the intensity of the normal map texture contribution.\r\nZero equals no normal map texture contribution."
   group=04_Layer4
   parameterName=Layer4_NormalMap_Intensity
   sortPriority=310
  }
  n55 = StaticComponentMaskParameter(None=@n52.Z Spec) {
   defaultR=true
   desc="Define the SpecularRoughnessMetallic_Texture channel used for the Specular mask.\r\nDefault is red.\r\nNot used on mobile renderer."
   group=04_Layer4
   parameterName=Layer4_SpecularChannel
   sortPriority=210
  }
  n56 = StaticComponentMaskParameter(None=@n52.Z Spec) {
   defaultB=true
   desc="Define the SpecularRoughnessMetallic_Texture channel used for the Specular mask.\r\nDefault is red.\r\nNot used on mobile renderer."
   group=04_Layer4
   parameterName=Layer4_RoughnessChannel
   sortPriority=220
  }
  n57 = StaticComponentMaskParameter(None=@n52.Z Spec) {
   defaultG=true
   desc="Define the SpecularRoughnessMetallic_Texture channel used for the Specular mask.\r\nDefault is red.\r\nNot used on mobile renderer."
   group=04_Layer4
   parameterName=Layer4_MetallicChannel
   sortPriority=230
  }
  n58 = VectorParameter() {
   desc="Designate which Basecolor texture channel to use for the heightmap. The heightmap will affect how this layer blends with other layers.\r\nSet all channels to zero, then set the channel you want to use to one.\r\nYou may get interesting results by playing with all these values.\r\nIf the texture doesn't have an alpha channel, leave the alpha value at zero."
   group=04_Layer4
   parameterName=Layer4_BaseColor_Heightmap_Channel
   sortPriority=120
   defaultValue=<rgba 0 0 0 1>
  }
  n59 = AppendVector(A=@n58.RGB, B=@n58.RGB)
  n60 = ScalarParameter() {
   defaultValue=1
   desc="Zero equals no heightmap contribution.\r\nThis value is clamped 0-1."
   group=04_Layer4
   parameterName=Layer4_BaseColor_Heightmap_Contribution
   sortPriority=130
  }
  n61 = LinearInterpolate(B=Saturate(None=DotProduct(A=@n52.Z Diffuse Texture, B=@n59)), Alpha=Saturate(None=@n60)) {constA=0.5}
  n62 = SetMaterialAttributes(
   Base Color=@n53
   Normal=LinearInterpolate(A=VertexNormalWS(), B=@n52.Normal Z Texture, Alpha=Abs(None=@n54))
   Specular=@n55
   Roughness=@n56
   Metallic=@n57
   Opacity Mask=@n61
  ) {attributeSetTypes [
    #2
    #3
    #4
    #5
    #6
    #1
   ]}
  n63 = ScalarParameter() {
   defaultValue=0.5
   desc="Mobile renderer doesn't use a texture, only this scalar value."
   group=04_Layer4
   parameterName=Layer4_Mobile_SpecularValue
   sortPriority=240
  }
  n64 = ScalarParameter() {
   defaultValue=0.85
   desc="Mobile renderer doesn't use a texture, only this scalar value."
   group=04_Layer4
   parameterName=Layer4_Mobile_RoughnessValue
   sortPriority=250
  }
  n65 = ScalarParameter() {
   desc="Mobile renderer doesn't use a texture, only this scalar value."
   group=04_Layer4
   parameterName=Layer4_Mobile_MetallicValue
   sortPriority=260
  }
  n66 = SetMaterialAttributes(Base Color=@n53, Normal=VertexNormalWS(), Specular=@n63, Roughness=@n64, Metallic=@n65, Opacity Mask=@n61) {attributeSetTypes [
    #2
    #3
    #4
    #5
    #6
    #1
   ]}
  NamedRerouteDeclaration(None=MaterialFunctionCall(
    Low _ PC Low and Med=@n62
    "High _ XB1, PS4"=@n62
    "Epic _ PC High, XSX, PS5"=@n62
    "Mobile Low _ PC Perf, Android"=@n66
    Mobile Med _ Switch=@n66
   ) {materialFunction=&MF_Fortnite_QualitySwitch_MaterialAttributes}.Result) {name=Layer4_Material, nodeColor=<rgba 0.5 0 1 1>}
  n67 = ScalarParameter() {
   defaultValue=1
   desc=A higher value will sharpen the blending to other layers.
   group=04_Layer4
   parameterName=Layer4_BaseColor_Heightmap_LayerBlendContrast
   sortPriority=140
  }
  NamedRerouteDeclaration(None=Abs(None=@n67)) {name=Layer4_HeightBlendContrast, nodeColor=<rgba 0.5 0 1 1>}
  n68 = VectorParameter() {
   desc="Divide the BaseColor_Texture by this color to adjust the color values.\r\nA vaule of 1,1,1 will not alter the colors.\r\nAlpha value is not used."
   group=05_Layer5
   parameterName=Layer5_BaseColor_Tint
   sortPriority=110
   channelNames {a=Not Used, b=Blue Value, g=Green Value, r=Red Value}
   defaultValue=<rgba 1 1 1 1>
  }
  n69 = TextureObjectParameter() {
   desc="BaseColor texture using Default compression.\r\nUse a heightmap in the alpha channel for more accurate layer blending results."
   group=05_Layer5
   parameterName=Layer5_BaseColor_Texture
   sortPriority=100
   texture=&T_Default_BaseColor_White
  }
  n70 = TextureObjectParameter() {
   desc="Channel-packed texture using Masks compression.\r\nOptimal default setup: Specular uses Red channel, Roughness uses Green channel, Metallic uses Blue channel.\r\nMany of the older Fortnite textures use green channel for metallic and blue channel for roughness.\r\nLeave the default setting if not used."
   group=05_Layer5
   parameterName=Layer5_SpecularRoughnessMetallic_Texture
   samplerType=SAMPLERTYPE_Masks
   sortPriority=200
   texture=&T_Default_Masks_SpecRoughMetal
  }
  n71 = ScalarParameter() {
   defaultValue=900
   desc=Scale of the textures in the level. A lower value decreases the size.
   group=05_Layer5
   parameterName=Layer5_TextureScale
   sortPriority=90
  }
  n72 = TextureObjectParameter() {
   desc="Tangent-space normal map texture using Normalmap compression.\r\nLeave the default setting if not used."
   group=05_Layer5
   parameterName=Layer5_NormalMap_Texture
   samplerType=SAMPLERTYPE_Normal
   sortPriority=300
   texture=&T_Default_NormalMap
  }
  n73 = MaterialFunctionCall(
   Diffuse Texture Object=@n69
   Spec Texture Object=@n70
   TextureSize=@n71
   NormalTextureObject=@n72
   Export Float 4 Diffuse=StaticBool() {value=true}
  ) {materialFunction=&WorldAlignedTextures_Complex}
  n74 = Saturate(None=Multiply(A=ComponentMask(None=@n68) {b=true, g=true, r=true}, B=ComponentMask(None=@n73.Z Diffuse Texture) {b=true, g=true, r=true}))
  n75 = ScalarParameter() {
   defaultValue=1
   desc="Controls the intensity of the normal map texture contribution.\r\nZero equals no normal map texture contribution."
   group=05_Layer5
   parameterName=Layer5_NormalMap_Intensity
   sortPriority=310
  }
  n76 = StaticComponentMaskParameter(None=@n73.Z Spec) {
   defaultR=true
   desc="Define the SpecularRoughnessMetallic_Texture channel used for the Specular mask.\r\nDefault is red.\r\nNot used on mobile renderer."
   group=05_Layer5
   parameterName=Layer5_SpecularChannel
   sortPriority=210
  }
  n77 = StaticComponentMaskParameter(None=@n73.Z Spec) {
   defaultB=true
   desc="Define the SpecularRoughnessMetallic_Texture channel used for the Specular mask.\r\nDefault is red.\r\nNot used on mobile renderer."
   group=05_Layer5
   parameterName=Layer5_RoughnessChannel
   sortPriority=220
  }
  n78 = StaticComponentMaskParameter(None=@n73.Z Spec) {
   defaultG=true
   desc="Define the SpecularRoughnessMetallic_Texture channel used for the Specular mask.\r\nDefault is red.\r\nNot used on mobile renderer."
   group=05_Layer5
   parameterName=Layer5_MetallicChannel
   sortPriority=230
  }
  n79 = VectorParameter() {
   desc="Designate which Basecolor texture channel to use for the heightmap. The heightmap will affect how this layer blends with other layers.\r\nSet all channels to zero, then set the channel you want to use to one.\r\nYou may get interesting results by playing with all these values.\r\nIf the texture doesn't have an alpha channel, leave the alpha value at zero."
   group=05_Layer5
   parameterName=Layer5_BaseColor_Heightmap_Channel
   sortPriority=120
   defaultValue=<rgba 0 0 0 1>
  }
  n80 = AppendVector(A=@n79.RGB, B=@n79.RGB)
  n81 = ScalarParameter() {
   defaultValue=1
   desc="Zero equals no heightmap contribution.\r\nThis value is clamped 0-1."
   group=05_Layer5
   parameterName=Layer5_BaseColor_Heightmap_Contribution
   sortPriority=130
  }
  n82 = LinearInterpolate(B=Saturate(None=DotProduct(A=@n73.Z Diffuse Texture, B=@n80)), Alpha=Saturate(None=@n81)) {constA=0.5}
  n83 = SetMaterialAttributes(
   Base Color=@n74
   Normal=LinearInterpolate(A=VertexNormalWS(), B=@n73.Normal Z Texture, Alpha=Abs(None=@n75))
   Specular=@n76
   Roughness=@n77
   Metallic=@n78
   Opacity Mask=@n82
  ) {attributeSetTypes [
    #2
    #3
    #4
    #5
    #6
    #1
   ]}
  n84 = ScalarParameter() {
   defaultValue=0.5
   desc="Mobile renderer doesn't use a texture, only this scalar value."
   group=05_Layer5
   parameterName=Layer5_Mobile_SpecularValue
   sortPriority=240
  }
  n85 = ScalarParameter() {
   defaultValue=0.85
   desc="Mobile renderer doesn't use a texture, only this scalar value."
   group=05_Layer5
   parameterName=Layer5_Mobile_RoughnessValue
   sortPriority=250
  }
  n86 = ScalarParameter() {
   desc="Mobile renderer doesn't use a texture, only this scalar value."
   group=05_Layer5
   parameterName=Layer5_Mobile_MetallicValue
   sortPriority=260
  }
  n87 = SetMaterialAttributes(Base Color=@n74, Normal=VertexNormalWS(), Specular=@n84, Roughness=@n85, Metallic=@n86, Opacity Mask=@n82) {attributeSetTypes [
    #2
    #3
    #4
    #5
    #6
    #1
   ]}
  NamedRerouteDeclaration(None=MaterialFunctionCall(
    Low _ PC Low and Med=@n83
    "High _ XB1, PS4"=@n83
    "Epic _ PC High, XSX, PS5"=@n83
    "Mobile Low _ PC Perf, Android"=@n87
    Mobile Med _ Switch=@n87
   ) {materialFunction=&MF_Fortnite_QualitySwitch_MaterialAttributes}.Result) {name=Layer5_Material, nodeColor=<rgba 0.5 0 1 1>}
  n88 = ScalarParameter() {
   defaultValue=1
   desc=A higher value will sharpen the blending to other layers.
   group=05_Layer5
   parameterName=Layer5_BaseColor_Heightmap_LayerBlendContrast
   sortPriority=140
  }
  NamedRerouteDeclaration(None=Abs(None=@n88)) {name=Layer5_HeightBlendContrast, nodeColor=<rgba 0.5 0 1 1>}
  n89 = VectorParameter() {
   desc="Divide the BaseColor_Texture by this color to adjust the color values.\r\nA vaule of 1,1,1 will not alter the colors.\r\nAlpha value is not used."
   group=06_Layer6
   parameterName=Layer6_BaseColor_Tint
   sortPriority=110
   channelNames {a=Not Used, b=Blue Value, g=Green Value, r=Red Value}
   defaultValue=<rgba 1 1 1 1>
  }
  n90 = TextureObjectParameter() {
   desc="BaseColor texture using Default compression.\r\nUse a heightmap in the alpha channel for more accurate layer blending results."
   group=06_Layer6
   parameterName=Layer6_BaseColor_Texture
   sortPriority=100
   texture=&T_Default_BaseColor_White
  }
  n91 = TextureObjectParameter() {
   desc="Channel-packed texture using Masks compression.\r\nOptimal default setup: Specular uses Red channel, Roughness uses Green channel, Metallic uses Blue channel.\r\nMany of the older Fortnite textures use green channel for metallic and blue channel for roughness.\r\nLeave the default setting if not used."
   group=06_Layer6
   parameterName=Layer6_SpecularRoughnessMetallic_Texture
   samplerType=SAMPLERTYPE_Masks
   sortPriority=200
   texture=&T_Default_Masks_SpecRoughMetal
  }
  n92 = ScalarParameter() {
   defaultValue=900
   desc=Scale of the textures in the level. A lower value decreases the size.
   group=06_Layer6
   parameterName=Layer6_TextureScale
   sortPriority=90
  }
  n93 = TextureObjectParameter() {
   desc="Tangent-space normal map texture using Normalmap compression.\r\nLeave the default setting if not used."
   group=06_Layer6
   parameterName=Layer6_NormalMap_Texture
   samplerType=SAMPLERTYPE_Normal
   sortPriority=300
   texture=&T_Default_NormalMap
  }
  n94 = MaterialFunctionCall(
   Diffuse Texture Object=@n90
   Spec Texture Object=@n91
   TextureSize=@n92
   NormalTextureObject=@n93
   Export Float 4 Diffuse=StaticBool() {value=true}
  ) {materialFunction=&WorldAlignedTextures_Complex}
  n95 = Saturate(None=Multiply(A=ComponentMask(None=@n89) {b=true, g=true, r=true}, B=ComponentMask(None=@n94.Z Diffuse Texture) {b=true, g=true, r=true}))
  n96 = ScalarParameter() {
   defaultValue=1
   desc="Controls the intensity of the normal map texture contribution.\r\nZero equals no normal map texture contribution."
   group=06_Layer6
   parameterName=Layer6_NormalMap_Intensity
   sortPriority=310
  }
  n97 = StaticComponentMaskParameter(None=@n94.Z Spec) {
   defaultR=true
   desc="Define the SpecularRoughnessMetallic_Texture channel used for the Specular mask.\r\nDefault is red.\r\nNot used on mobile renderer."
   group=06_Layer6
   parameterName=Layer6_SpecularChannel
   sortPriority=210
  }
  n98 = StaticComponentMaskParameter(None=@n94.Z Spec) {
   defaultB=true
   desc="Define the SpecularRoughnessMetallic_Texture channel used for the Specular mask.\r\nDefault is red.\r\nNot used on mobile renderer."
   group=06_Layer6
   parameterName=Layer6_RoughnessChannel
   sortPriority=220
  }
  n99 = StaticComponentMaskParameter(None=@n94.Z Spec) {
   defaultG=true
   desc="Define the SpecularRoughnessMetallic_Texture channel used for the Specular mask.\r\nDefault is red.\r\nNot used on mobile renderer."
   group=06_Layer6
   parameterName=Layer6_MetallicChannel
   sortPriority=230
  }
  n100 = VectorParameter() {
   desc="Designate which Basecolor texture channel to use for the heightmap. The heightmap will affect how this layer blends with other layers.\r\nSet all channels to zero, then set the channel you want to use to one.\r\nYou may get interesting results by playing with all these values.\r\nIf the texture doesn't have an alpha channel, leave the alpha value at zero."
   group=06_Layer6
   parameterName=Layer6_BaseColor_Heightmap_Channel
   sortPriority=120
   defaultValue=<rgba 0 0 0 1>
  }
  n101 = AppendVector(A=@n100.RGB, B=@n100.RGB)
  n102 = ScalarParameter() {
   defaultValue=1
   desc="Zero equals no heightmap contribution.\r\nThis value is clamped 0-1."
   group=06_Layer6
   parameterName=Layer6_BaseColor_Heightmap_Contribution
   sortPriority=130
  }
  n103 = LinearInterpolate(B=Saturate(None=DotProduct(A=@n94.Z Diffuse Texture, B=@n101)), Alpha=Saturate(None=@n102)) {constA=0.5}
  n104 = SetMaterialAttributes(
   Base Color=@n95
   Normal=LinearInterpolate(A=VertexNormalWS(), B=@n94.Normal Z Texture, Alpha=Abs(None=@n96))
   Specular=@n97
   Roughness=@n98
   Metallic=@n99
   Opacity Mask=@n103
  ) {attributeSetTypes [
    #2
    #3
    #4
    #5
    #6
    #1
   ]}
  n105 = ScalarParameter() {
   defaultValue=0.5
   desc="Mobile renderer doesn't use a texture, only this scalar value."
   group=06_Layer6
   parameterName=Layer6_Mobile_SpecularValue
   sortPriority=240
  }
  n106 = ScalarParameter() {
   defaultValue=0.85
   desc="Mobile renderer doesn't use a texture, only this scalar value."
   group=06_Layer6
   parameterName=Layer6_Mobile_RoughnessValue
   sortPriority=250
  }
  n107 = ScalarParameter() {
   desc="Mobile renderer doesn't use a texture, only this scalar value."
   group=06_Layer6
   parameterName=Layer6_Mobile_MetallicValue
   sortPriority=260
  }
  n108 = SetMaterialAttributes(Base Color=@n95, Normal=VertexNormalWS(), Specular=@n105, Roughness=@n106, Metallic=@n107, Opacity Mask=@n103) {attributeSetTypes [
    #2
    #3
    #4
    #5
    #6
    #1
   ]}
  NamedRerouteDeclaration(None=MaterialFunctionCall(
    Low _ PC Low and Med=@n104
    "High _ XB1, PS4"=@n104
    "Epic _ PC High, XSX, PS5"=@n104
    "Mobile Low _ PC Perf, Android"=@n108
    Mobile Med _ Switch=@n108
   ) {materialFunction=&MF_Fortnite_QualitySwitch_MaterialAttributes}.Result) {name=Layer6_Material, nodeColor=<rgba 0.5 0 1 1>}
  n109 = ScalarParameter() {
   defaultValue=1
   desc=A higher value will sharpen the blending to other layers.
   group=06_Layer6
   parameterName=Layer6_BaseColor_Heightmap_LayerBlendContrast
   sortPriority=140
  }
  NamedRerouteDeclaration(None=Abs(None=@n109)) {name=Layer6_HeightBlendContrast, nodeColor=<rgba 0.5 0 1 1>}
  n110 = NamedRerouteUsage()
  n111 = MaterialFunctionCall!(
   CurrentMaterial=@n110
   AddMaterial=NamedRerouteUsage()
   LandscapeLayer=LandscapeLayerSample() {parameterName=Layer1}
   UseContrast=StaticBool() {value=true}
   Contrast=NamedRerouteUsage()
  ) {materialFunction=/Game/Creative/Materials/Landscape/Creative_Landscape_Apollo/MaterialFunctions/MF_BlendLandscapeLayers_Opacity}
  n112 = LandscapeLayerSwitch!(LayerUsed=@n111.BlendedMaterial, LayerNotUsed=@n110) {
   desc=The landscape layer switch allows the material to be less expensive if the designated layer is not painted on a landscape component.
   parameterName=Layer1
  }
  n113 = MaterialFunctionCall(
   CurrentMaterial=@n112
   AddMaterial=NamedRerouteUsage()
   LandscapeLayer=LandscapeLayerSample() {parameterName=Layer2}
   UseContrast=StaticBool() {value=true}
   Contrast=NamedRerouteUsage()
  ) {materialFunction=&MF_BlendLandscapeLayers_Opacity}
  n114 = LandscapeLayerSwitch(LayerUsed=@n113.BlendedMaterial, LayerNotUsed=@n112) {
   desc=The landscape layer switch allows the material to be less expensive if the designated layer is not painted on a landscape component.
   parameterName=Layer2
  }
  n115 = MaterialFunctionCall(
   CurrentMaterial=@n114
   AddMaterial=NamedRerouteUsage()
   LandscapeLayer=LandscapeLayerSample() {parameterName=Layer3}
   UseContrast=StaticBool() {value=true}
   Contrast=NamedRerouteUsage()
  ) {materialFunction=&MF_BlendLandscapeLayers_Opacity}
  n116 = LandscapeLayerSwitch(LayerUsed=@n115.BlendedMaterial, LayerNotUsed=@n114) {
   desc=The landscape layer switch allows the material to be less expensive if the designated layer is not painted on a landscape component.
   parameterName=Layer3
  }
  n117 = MaterialFunctionCall(
   CurrentMaterial=@n116
   AddMaterial=NamedRerouteUsage()
   LandscapeLayer=LandscapeLayerSample() {parameterName=Layer4}
   UseContrast=StaticBool() {value=true}
   Contrast=NamedRerouteUsage()
  ) {materialFunction=&MF_BlendLandscapeLayers_Opacity}
  n118 = LandscapeLayerSwitch(LayerUsed=@n117.BlendedMaterial, LayerNotUsed=@n116) {
   desc=The landscape layer switch allows the material to be less expensive if the designated layer is not painted on a landscape component.
   parameterName=Layer4
  }
  n119 = MaterialFunctionCall(
   CurrentMaterial=@n118
   AddMaterial=NamedRerouteUsage()
   LandscapeLayer=LandscapeLayerSample() {parameterName=Layer5}
   UseContrast=StaticBool() {value=true}
   Contrast=NamedRerouteUsage()
  ) {materialFunction=&MF_BlendLandscapeLayers_Opacity}
  n120 = LandscapeLayerSwitch(LayerUsed=@n119.BlendedMaterial, LayerNotUsed=@n118) {
   desc=The landscape layer switch allows the material to be less expensive if the designated layer is not painted on a landscape component.
   parameterName=Layer5
  }
  n121 = MaterialFunctionCall(
   CurrentMaterial=@n120
   AddMaterial=NamedRerouteUsage()
   LandscapeLayer=LandscapeLayerSample() {parameterName=Layer6}
   UseContrast=StaticBool() {value=true}
   Contrast=NamedRerouteUsage()
  ) {materialFunction=&MF_BlendLandscapeLayers_Opacity}
  n122 = LandscapeLayerSwitch(LayerUsed=@n121.BlendedMaterial, LayerNotUsed=@n120) {
   desc=The landscape layer switch allows the material to be less expensive if the designated layer is not painted on a landscape component.
   parameterName=Layer6
  }
  n123 = MaterialFunctionCall(
   CurrentMaterial=@n122
   AddMaterial=NamedRerouteUsage()
   LandscapeLayer=NamedRerouteUsage()
   UseContrast=StaticBool() {value=true}
   Contrast=NamedRerouteUsage()
  ) {materialFunction=&MF_BlendLandscapeLayers_Opacity}
  n124 = StaticBoolParameter() {
   defaultValue=true
   desc="Use Cliffs on the more vertical areas of the landscape.\r\nThis is typically used as an indicator of where the player is not allowed to traverse the landscape."
   group=Cliff
   parameterName=Cliff_UseCliff
   sortPriority=-100
  }
  n125 = StaticSwitch(True=@n123.BlendedMaterial, False=@n122, Value=@n124)
  n126 = GetMaterialAttributes(None=@n125) {attributeGetTypes [
    #2
    #3
    #4
    #5
    #6
   ]}
  NamedRerouteDeclaration(None=ComponentMask(None=@n126.Base Color) {b=true, g=true, r=true}) {name=Layer7_Basecolor_In, nodeColor=<rgba 0.5 0 1 1>}
  NamedRerouteDeclaration(None=@n126.Roughness) {name=Layer7_Rough_In, nodeColor=<rgba 0.5 0 1 1>}
  n127 = OneMinus(None=LandscapeLayerSample() {parameterName=Layer7})
  n128 = NamedRerouteUsage()
  n129 = ScalarParameter() {
   defaultValue=1
   desc=A higher value will sharpen the blending to other layers.
   group=07_Layer7
   parameterName=Layer7_HeightLerpContrast
   sortPriority=0
  }
  n130 = Abs(None=@n129)
  n131 = ScalarParameter() {
   defaultValue=1
   desc="With this layer, you can adjust all other layers where it is painted.\r\nExample: Wet areas, splattered paint, scorch marks.\r\nA value of zero will ignore Layer7.\r\nThis value is clamped 0-1."
   group=07_Layer7
   parameterName=Layer7_Amount
   sortPriority=-1
  }
  NamedRerouteDeclaration(None=LinearInterpolate(
    A=Constant() {r=1}
    B=MaterialFunctionCall(Transition Phase=@n127, Height Texture=@n128, Contrast=@n130) {materialFunction=&HeightLerp}.Alpha
    Alpha=Saturate(None=@n131)
   ) {constA=1}) {name=Layer7Mask, nodeColor=<rgba 0.5 0 1 1>}
  n132 = VectorParameter() {
   desc="Divide the BaseColor_Texture by this color to adjust the color values.\r\nA vaule of 1,1,1 will not alter the colors.\r\nAlpha value is not used."
   group=03_Layer3
   parameterName=Layer3_BaseColor_Tint
   sortPriority=110
   channelNames {a=Not Used, b=Blue Value, g=Green Value, r=Red Value}
   defaultValue=<rgba 1 1 1 1>
  }
  n133 = TextureObjectParameter() {
   desc="BaseColor texture using Default compression.\r\nUse a heightmap in the alpha channel for more accurate layer blending results."
   group=03_Layer3
   parameterName=Layer3_BaseColor_Texture
   sortPriority=100
   texture=&T_Default_BaseColor_White
  }
  n134 = TextureObjectParameter() {
   desc="Channel-packed texture using Masks compression.\r\nOptimal default setup: Specular uses Red channel, Roughness uses Green channel, Metallic uses Blue channel.\r\nMany of the older Fortnite textures use green channel for metallic and blue channel for roughness.\r\nLeave the default setting if not used."
   group=03_Layer3
   parameterName=Layer3_SpecularRoughnessMetallic_Texture
   samplerType=SAMPLERTYPE_Masks
   sortPriority=200
   texture=&T_Default_Masks_SpecRoughMetal
  }
  n135 = ScalarParameter() {
   defaultValue=900
   desc=Scale of the textures in the level. A lower value decreases the size.
   group=03_Layer3
   parameterName=Layer3_TextureScale
   sortPriority=90
  }
  n136 = TextureObjectParameter() {
   desc="Tangent-space normal map texture using Normalmap compression.\r\nLeave the default setting if not used."
   group=03_Layer3
   parameterName=Layer3_NormalMap_Texture
   samplerType=SAMPLERTYPE_Normal
   sortPriority=300
   texture=&T_Default_NormalMap
  }
  n137 = MaterialFunctionCall(
   Diffuse Texture Object=@n133
   Spec Texture Object=@n134
   TextureSize=@n135
   NormalTextureObject=@n136
   Export Float 4 Diffuse=StaticBool() {value=true}
  ) {materialFunction=&WorldAlignedTextures_Complex}
  n138 = Saturate(None=Multiply(A=ComponentMask(None=@n132) {b=true, g=true, r=true}, B=ComponentMask(None=@n137.Z Diffuse Texture) {b=true, g=true, r=true}))
  n139 = ScalarParameter() {
   defaultValue=1
   desc="Controls the intensity of the normal map texture contribution.\r\nZero equals no normal map texture contribution."
   group=03_Layer3
   parameterName=Layer3_NormalMap_Intensity
   sortPriority=310
  }
  n140 = StaticComponentMaskParameter(None=@n137.Z Spec) {
   defaultR=true
   desc="Define the SpecularRoughnessMetallic_Texture channel used for the Specular mask.\r\nDefault is red.\r\nNot used on mobile renderer."
   group=03_Layer3
   parameterName=Layer3_SpecularChannel
   sortPriority=210
  }
  n141 = StaticComponentMaskParameter(None=@n137.Z Spec) {
   defaultB=true
   desc="Define the SpecularRoughnessMetallic_Texture channel used for the Specular mask.\r\nDefault is red.\r\nNot used on mobile renderer."
   group=03_Layer3
   parameterName=Layer3_RoughnessChannel
   sortPriority=220
  }
  n142 = StaticComponentMaskParameter(None=@n137.Z Spec) {
   defaultG=true
   desc="Define the SpecularRoughnessMetallic_Texture channel used for the Specular mask.\r\nDefault is red.\r\nNot used on mobile renderer."
   group=03_Layer3
   parameterName=Layer3_MetallicChannel
   sortPriority=230
  }
  n143 = VectorParameter() {
   desc="Designate which Basecolor texture channel to use for the heightmap. The heightmap will affect how this layer blends with other layers.\r\nSet all channels to zero, then set the channel you want to use to one.\r\nYou may get interesting results by playing with all these values.\r\nIf the texture doesn't have an alpha channel, leave the alpha value at zero."
   group=03_Layer3
   parameterName=Layer3_BaseColor_Heightmap_Channel
   sortPriority=120
   defaultValue=<rgba 0 0 0 1>
  }
  n144 = AppendVector(A=@n143.RGB, B=@n143.RGB)
  n145 = ScalarParameter() {
   defaultValue=1
   desc="Zero equals no heightmap contribution.\r\nThis value is clamped 0-1."
   group=03_Layer3
   parameterName=Layer3_BaseColor_Heightmap_Contribution
   sortPriority=130
  }
  n146 = LinearInterpolate(B=Saturate(None=DotProduct(A=@n137.Z Diffuse Texture, B=@n144)), Alpha=Saturate(None=@n145)) {constA=0.5}
  n147 = SetMaterialAttributes(
   Base Color=@n138
   Normal=LinearInterpolate(A=VertexNormalWS(), B=@n137.Normal Z Texture, Alpha=Abs(None=@n139))
   Specular=@n140
   Roughness=@n141
   Metallic=@n142
   Opacity Mask=@n146
  ) {attributeSetTypes [
    #2
    #3
    #4
    #5
    #6
    #1
   ]}
  n148 = ScalarParameter() {
   defaultValue=0.5
   desc="Mobile renderer doesn't use a texture, only this scalar value."
   group=04_Layer4
   parameterName=Layer3_Mobile_SpecularValue
   sortPriority=240
  }
  n149 = ScalarParameter() {
   defaultValue=0.85
   desc="Mobile renderer doesn't use a texture, only this scalar value."
   group=03_Layer3
   parameterName=Layer3_Mobile_RoughnessValue
   sortPriority=250
  }
  n150 = ScalarParameter() {
   desc="Mobile renderer doesn't use a texture, only this scalar value."
   group=03_Layer3
   parameterName=Layer3_Mobile_MetallicValue
   sortPriority=260
  }
  n151 = SetMaterialAttributes(Base Color=@n138, Normal=VertexNormalWS(), Specular=@n148, Roughness=@n149, Metallic=@n150, Opacity Mask=@n146) {attributeSetTypes [
    #2
    #3
    #4
    #5
    #6
    #1
   ]}
  NamedRerouteDeclaration(None=MaterialFunctionCall(
    Low _ PC Low and Med=@n147
    "High _ XB1, PS4"=@n147
    "Epic _ PC High, XSX, PS5"=@n147
    "Mobile Low _ PC Perf, Android"=@n151
    Mobile Med _ Switch=@n151
   ) {materialFunction=&MF_Fortnite_QualitySwitch_MaterialAttributes}.Result) {name=Layer3_Material, nodeColor=<rgba 0.5 0 1 1>}
  NamedRerouteDeclaration(None=@n126.Normal) {name=Layer7_Normals_In, nodeColor=<rgba 0.5 0 1 1>}
  NamedRerouteDeclaration(None=@n126.Specular) {name=Layer7_Spec_In, nodeColor=<rgba 0.5 0 1 1>}
  NamedRerouteDeclaration(None=@n126.Metallic) {name=Layer7_Metal_In, nodeColor=<rgba 0.5 0 1 1>}
  n152 = NamedRerouteUsage()
  n153 = TextureObjectParameter() {
   desc="BaseColor texture using Default compression.\r\nUse a heightmap in the alpha channel for more accurate layer blending results."
   group=00_Base
   parameterName=Base_BaseColor_Texture
   sortPriority=-1002
   texture=/Game/Environments/Apollo/Sets/Terrain/Textures/T_Apollo_Grass_D_01
  }
  n154 = TextureObjectParameter() {
   desc="Channel-packed texture using Masks compression.\r\nOptimal default setup: Specular uses Red channel, Roughness uses Green channel, Metallic uses Blue channel.\r\nMany of the older Fortnite textures use green channel for metallic and blue channel for roughness."
   group=00_Base
   parameterName=Base_SpecularRoughnessMetallic_Texture
   samplerType=SAMPLERTYPE_Masks
   sortPriority=200
   texture=/Game/Environments/Apollo/Sets/Terrain/Textures/T_Apollo_Grass_S_01
  }
  n155 = ScalarParameter() {
   defaultValue=900
   desc=Scale of the textures in the level. A lower value decreases the size.
   group=00_Base
   parameterName=Base_TextureScale
   sortPriority=-1003
  }
  n156 = TextureObjectParameter() {
   desc=Tangent-space normal map texture using Normalmap compression.
   group=00_Base
   parameterName=Base_NormalMap_Texture
   samplerType=SAMPLERTYPE_Normal
   sortPriority=300
   texture=/Game/Environments/Apollo/Sets/Terrain/Textures/T_Apollo_Grass_N_01
  }
  n157 = MaterialFunctionCall(
   Diffuse Texture Object=@n153
   Spec Texture Object=@n154
   TextureSize=@n155
   NormalTextureObject=@n156
   Export Float 4 Diffuse=StaticBool() {value=true}
  ) {materialFunction=&WorldAlignedTextures_Complex}
  n158 = ScalarParameter() {
   defaultValue=1
   desc="Controls the intensity of the normal map texture contribution.\r\nZero equals no normal map texture contribution."
   group=00_Base
   parameterName=Base_NormalMap_Intensity
   sortPriority=310
  }
  n159 = Abs(None=@n158)
  n160 = ScalarParameter() {
   defaultValue=15000
   desc="This gives the NormalMap a flat appearance in the distance.\r\nThe smaller the value, the closer to the camera the fade will be."
   group=00_Base
   parameterName=Base_NormalMap_FadeDistance
   sortPriority=320
  }
  n161 = ScalarParameter() {
   defaultValue=1
   desc="A value of zero will ignore FadeDistance.\r\nThis value is clamped 0-1."
   group=00_Base
   parameterName=Base_NormalMap_FadeDistanceAmount
   sortPriority=330
  }
  n162 = LinearInterpolate(B=SphereMask(A=WorldPosition().XYZ, B=CameraPositionWS(), Radius=@n160) {hardnessPercent=0}, Alpha=Saturate(None=@n161)) {constA=1}
  n163 = StaticComponentMaskParameter(None=@n157.Z Spec) {
   defaultR=true
   desc="Define the SpecularRoughnessMetallic_Texture channel used for the Specular mask.\r\nDefault is red.\r\nNot used on mobile renderer."
   group=00_Base
   parameterName=Base_SpecularChannel
   sortPriority=210
  }
  n164 = StaticComponentMaskParameter(None=@n157.Z Spec) {
   defaultB=true
   desc="Define the SpecularRoughnessMetallic_Texture channel used for the Roughness mask.\r\nDefault is green.\r\nMany of the older Fortnite textures use green channel for metallic and blue channel for roughness.\r\nNot used on mobile renderer."
   group=00_Base
   parameterName=Base_RoughnessChannel
   sortPriority=220
  }
  n165 = StaticComponentMaskParameter(None=@n157.Z Spec) {
   defaultG=true
   desc="Define the SpecularRoughnessMetallic_Texture channel used for the Metallic mask.\r\nDefault is blue.\r\nMany of the older Fortnite textures use green channel for metallic and blue channel for roughness.\r\nNot used on mobile renderer."
   group=00_Base
   parameterName=Base_MetallicChannel
   sortPriority=230
  }
  n166 = VectorParameter() {
   desc="Designate which Basecolor texture channel to use for the heightmap. The heightmap will affect how this layer blends with other layers.\r\nSet all channels to zero, then set the channel you want to use to one.\r\nYou may get interesting results by playing with all these values.\r\nIf the texture doesn't have an alpha channel, leave the alpha value at zero."
   group=00_Base
   parameterName=Base_BaseColor_Heightmap_Channel
   sortPriority=-1001
   defaultValue=<rgba 0 0 0 1>
  }
  n167 = AppendVector(A=@n166.RGB, B=@n166.RGB)
  n168 = ScalarParameter() {
   defaultValue=1
   desc="Zero equals no heightmap contribution.\r\nThis value is clamped 0-1."
   group=00_Base
   parameterName=Base_BaseColor_Heightmap_Contribution
   sortPriority=-1000
  }
  n169 = LinearInterpolate(B=Saturate(None=DotProduct(A=@n157.Z Diffuse Texture, B=@n167)), Alpha=Saturate(None=@n168)) {constA=0.5}
  n170 = SetMaterialAttributes(
   Base Color=@n152
   Normal=LinearInterpolate(A=VertexNormalWS(), B=@n157.Normal Z Texture, Alpha=Multiply(A=@n159, B=@n162))
   Specular=@n163
   Roughness=@n164
   Metallic=@n165
   Opacity Mask=@n169
  ) {attributeSetTypes [
    #2
    #3
    #4
    #5
    #6
    #1
   ]}
  n171 = ScalarParameter() {
   defaultValue=0.5
   desc="Mobile renderer doesn't use a texture, only this scalar value."
   group=00_Base
   parameterName=Base_Mobile_SpecularValue
   sortPriority=240
  }
  n172 = ScalarParameter() {
   defaultValue=0.9
   desc="Mobile renderer doesn't use a texture, only this scalar value."
   group=00_Base
   parameterName=Base_Mobile_RoughnessValue
   sortPriority=250
  }
  n173 = ScalarParameter() {
   desc="Mobile renderer doesn't use a texture, only this scalar value."
   group=00_Base
   parameterName=Base_Mobile_MetallicValue
   sortPriority=260
  }
  n174 = SetMaterialAttributes(Base Color=@n152, Normal=VertexNormalWS(), Specular=@n171, Roughness=@n172, Metallic=@n173, Opacity Mask=@n169) {attributeSetTypes [
    #2
    #3
    #4
    #5
    #6
    #1
   ]}
  NamedRerouteDeclaration(None=MaterialFunctionCall(
    Low _ PC Low and Med=@n170
    "High _ XB1, PS4"=@n170
    "Epic _ PC High, XSX, PS5"=@n170
    "Mobile Low _ PC Perf, Android"=@n174
    Mobile Med _ Switch=@n174
   ) {materialFunction=&MF_Fortnite_QualitySwitch_MaterialAttributes}.Result) {name=Base_Material, nodeColor=<rgba 0.5 0 1 1>}
  n175 = VectorParameter() {
   desc="Divide the BaseColor_Texture by this color to adjust the color values.\r\nA vaule of 1,1,1 will not alter the colors.\r\nAlpha value is not used."
   group=Cliff
   parameterName=Cliff_BaseColor_Tint
   sortPriority=110
   channelNames {a=Not Used, b=Blue Value, g=Green Value, r=Red Value}
   defaultValue=<rgba 1 1 1 1>
  }
  n176 = TextureObjectParameter() {
   desc="BaseColor texture using Default compression.\r\nUse a heightmap in the alpha channel for more accurate layer blending results."
   group=Cliff
   parameterName=Cliff_BaseColor_Texture
   sortPriority=100
   texture=&T_Default_BaseColor_White
  }
  n177 = TextureObjectParameter() {
   desc="Channel-packed texture using Masks compression.\r\nOptimal default setup: Specular uses Red channel, Roughness uses Green channel, Metallic uses Blue channel.\r\nMany of the older Fortnite textures use green channel for metallic and blue channel for roughness.\r\nLeave the default setting if not used."
   group=Cliff
   parameterName=Cliff_SpecularRoughnessMetallic_Texture
   samplerType=SAMPLERTYPE_Masks
   sortPriority=200
   texture=&T_Default_Masks_SpecRoughMetal
  }
  n178 = VectorParameter() {
   desc="Expand the dropdown for easier manipulation.\r\nRed=Width\r\nGreen=Depth\r\nBlue=Height\r\nAlpha=Scale multiplier"
   group=Cliff
   parameterName=Cliff_TextureScale
   sortPriority=90
   channelNames {a=ScaleMultiplier, b=Height, g=Depth, r=Width}
   defaultValue=<rgba 3500 3500 3000 1>
  }
  n179 = ScalarParameter() {
   defaultValue=0.5
   desc="Adjust the contrast of the X && Y projection mask."
   group=Cliff
   parameterName=Cliff_ProjectionMask_TransitionContrast
   sortPriority=93
  }
  n180 = StaticBool() {value=true}
  n181 = ScalarParameter() {
   defaultValue=0.5
   desc=Adjust the contrast of the projection heightmap transition.
   group=Cliff
   parameterName=Cliff_ProjectionMask_HeightmapTransitionContrast
   sortPriority=92
  }
  n182 = VectorParameter() {
   desc="Designate which BaseColor texture channel to use for the projection mask blending.\r\nUsusally the heightmap channel works best."
   group=Cliff
   parameterName=Cliff_ProjectionMask_BaseColorChannel
   sortPriority=91
   defaultValue=<rgba 0 0 0 1>
  }
  n183 = TextureObjectParameter() {
   desc="Tangent-space normal map texture using Normalmap compression.\r\nLeave the default setting if not used."
   group=Cliff
   parameterName=Cliff_NormalMap_Texture
   samplerType=SAMPLERTYPE_Normal
   sortPriority=300
   texture=&T_Default_NormalMap
  }
  n184 = StaticBool() {value=true}
  n185 = MaterialFunctionCall(
   Diffuse Texture Object=@n176
   Spec Texture Object=@n177
   TextureSize=Multiply(A=ComponentMask(None=@n178) {b=true, g=true, r=true}, B=@n178.ScaleMultiplier)
   ProjectionTransitionContrast=@n179
   Use Texture Alpha Heightmap to Refine the Transistion=@n180
   HeightLerpTransitionContrast=@n181
   PickTheDiffuseHeightChannel=@n180
   TheDiffuseHeightChannel=AppendVector(A=@n182.RGB, B=@n182.RGB)
   NormalTextureObject=@n183
   Export Float 4 Diffuse=@n184
   Export Float 4 Spec=@n184
  ) {materialFunction=&WorldAlignedTextures_Complex}
  n186 = Saturate(None=Multiply(A=ComponentMask(None=@n175) {b=true, g=true, r=true}, B=ComponentMask(None=@n185.XY Diffuse Texture) {b=true, g=true, r=true}))
  n187 = ScalarParameter() {
   defaultValue=1
   desc="Controls the intensity of the normal map texture contribution.\r\nZero equals no normal map texture contribution."
   group=Cliff
   parameterName=Cliff_NormalMap_Intensity
   sortPriority=310
  }
  n188 = StaticComponentMaskParameter(None=@n185.XY Spec) {
   defaultR=true
   desc="Define the SpecularRoughnessMetallic_Texture channel used for the Specular mask.\r\nDefault is red.\r\nNot used on mobile renderer."
   group=Cliff
   parameterName=Cliff_SpecularChannel
   sortPriority=210
  }
  n189 = ScalarParameter() {
   desc="Minimum roughness value.\r\nSet minimum && maximum to the same amount to get a singular metallic value."
   group=Cliff
   parameterName=Cliff_Roughness_Minimum
   sortPriority=221
  }
  n190 = ScalarParameter() {
   defaultValue=1
   desc="Maximum roughness value.\r\nSet minimum && maximum to the same amount to get a singular metallic value."
   group=Cliff
   parameterName=Cliff_Roughness_Maximum
   sortPriority=222
  }
  n191 = StaticComponentMaskParameter(None=@n185.XY Spec) {
   defaultB=true
   desc="Define the SpecularRoughnessMetallic_Texture channel used for the Specular mask.\r\nDefault is red.\r\nNot used on mobile renderer."
   group=Cliff
   parameterName=Cliff_RoughnessChannel
   sortPriority=220
  }
  n192 = ScalarParameter() {
   desc="Minimum metallic value.\r\nSet minimum && maximum to the same amount to get a singular metallic value."
   group=Cliff
   parameterName=Cliff_Metallic_Minimum
   sortPriority=231
  }
  n193 = ScalarParameter() {
   defaultValue=1
   desc="Maximum metallic value.\r\nSet minimum && maximum to the same amount to get a singular metallic value."
   group=Cliff
   parameterName=Cliff_Metallic_Maximum
   sortPriority=232
  }
  n194 = StaticComponentMaskParameter(None=@n185.XY Spec) {
   defaultG=true
   desc="Define the SpecularRoughnessMetallic_Texture channel used for the Specular mask.\r\nDefault is red.\r\nNot used on mobile renderer."
   group=Cliff
   parameterName=Cliff_MetallicChannel
   sortPriority=230
  }
  n195 = VectorParameter() {
   desc="Designate which Basecolor texture channel to use for the heightmap. The heightmap will affect how this layer blends with other layers.\r\nSet all channels to zero, then set the channel you want to use to one.\r\nYou may get interesting results by playing with all these values.\r\nIf the texture doesn't have an alpha channel, leave the alpha value at zero."
   group=Cliff
   parameterName=Cliff_BaseColor_Heightmap_Channel
   sortPriority=120
   defaultValue=<rgba 0 0 0 1>
  }
  n196 = AppendVector(A=@n195.RGB, B=@n195.RGB)
  n197 = ScalarParameter() {
   defaultValue=1
   desc="Zero equals no heightmap contribution.\r\nThis value is clamped 0-1."
   group=Cliff
   parameterName=Cliff_BaseColor_Heightmap_Contribution
   sortPriority=130
  }
  n198 = LinearInterpolate(B=Saturate(None=DotProduct(A=@n185.XY Diffuse Texture, B=@n196)), Alpha=Saturate(None=@n197)) {constA=0.5}
  n199 = SetMaterialAttributes(
   Base Color=@n186
   Normal=LinearInterpolate(A=VertexNormalWS(), B=@n185.Normal XY Texture, Alpha=Abs(None=@n187))
   Specular=@n188
   Roughness=Saturate(None=LinearInterpolate(A=@n189, B=@n190, Alpha=@n191))
   Metallic=Saturate(None=LinearInterpolate(A=@n192, B=@n193, Alpha=@n194))
   Opacity Mask=@n198
  ) {attributeSetTypes [
    #2
    #3
    #4
    #5
    #6
    #1
   ]}
  n200 = ScalarParameter() {
   defaultValue=0.5
   desc="Mobile renderer doesn't use a texture, only this scalar value."
   group=Cliff
   parameterName=Cliff_Mobile_SpecularValue
   sortPriority=240
  }
  n201 = ScalarParameter() {
   defaultValue=0.85
   desc="Mobile renderer doesn't use a texture, only this scalar value."
   group=Cliff
   parameterName=Cliff_Mobile_RoughnessValue
   sortPriority=250
  }
  n202 = ScalarParameter() {
   desc="Mobile renderer doesn't use a texture, only this scalar value."
   group=Cliff
   parameterName=Cliff_Mobile_MetallicValue
   sortPriority=260
  }
  n203 = SetMaterialAttributes(Base Color=@n186, Normal=VertexNormalWS(), Specular=@n200, Roughness=@n201, Metallic=@n202, Opacity Mask=@n198) {attributeSetTypes [
    #2
    #3
    #4
    #5
    #6
    #1
   ]}
  NamedRerouteDeclaration(None=MaterialFunctionCall(
    Low _ PC Low and Med=@n199
    "High _ XB1, PS4"=@n199
    "Epic _ PC High, XSX, PS5"=@n199
    "Mobile Low _ PC Perf, Android"=@n203
    Mobile Med _ Switch=@n203
   ) {materialFunction=&MF_Fortnite_QualitySwitch_MaterialAttributes}.Result) {name=Cliff_Material, nodeColor=<rgba 0.5 0 1 1>}
  n204 = ComponentMask(None=VertexNormalWS()) {b=true}
  n205 = ScalarParameter() {
   defaultValue=3
   desc=Higher values will increase the slope mask area.
   group=Cliff
   parameterName=Cliff_SlopeLayerMask_Power
   sortPriority=-95
  }
  n206 = ScalarParameter() {
   defaultValue=1.5
   desc=Lower values will increase the slope mask area.
   group=Cliff
   parameterName=Cliff_SlopeLayerMask_Multiply
   sortPriority=-90
  }
  n207 = Multiply(A=Power(Base=Saturate(None=@n204), Exp=@n205) {constExponent=4}, B=@n206)
  NamedRerouteDeclaration(None=OneMinus(None=Saturate(None=@n207))) {name=Cliff_Slo, nodeColor=<rgba 0.5 0 1 1>}
  n208 = ScalarParameter() {
   defaultValue=-1
   desc=Linear interpolates the texture mask from the low range to the high range.
   group=00_Base
   parameterName=Base_BaseColorNoise_C_LowRange
   sortPriority=-390
  }
  n209 = ScalarParameter() {
   defaultValue=3
   desc=Linear interpolates the texture mask from the low range to the high range.
   group=00_Base
   parameterName=Base_BaseColorNoise_C_HighRange
   sortPriority=-380
  }
  n210 = StaticBool() {value=true}
  n211 = TextureObjectParameter() {
   group=00_Base
   parameterName=Base_BaseColorNoise_Texture
   samplerType=SAMPLERTYPE_LinearColor
   sortPriority=-499
   texture=/Game/Athena/Environments/Landscape/Textures/T_Athena_Snow_01_M
  }
  n212 = ScalarParameter() {
   defaultValue=4100
   desc=Scale of the textures in the level. A lower value decreases the size.
   group=00_Base
   parameterName=Base_BaseColorNoise_C_TextureScale
   sortPriority=-410
  }
  n213 = StaticComponentMaskParameter(None=MaterialFunctionCall(Export Float 4=@n210, TextureObject=@n211, TextureSize=@n212) {materialFunction=/Game/Weapons/WeaponSkins/Wraps/MaterialFunctions/WorldAlignedTexture}.Z Texture) {
   defaultA=true
   desc=Define the texture channel to use for BaseColorNoise_C.
   group=00_Base
   parameterName=Base_BaseColorNoise_C_MaskChannel
   sortPriority=-400
  }
  NamedRerouteDeclaration(None=Saturate(None=LinearInterpolate(A=@n208, B=@n209, Alpha=@n213))) {name=Noise_C, nodeColor=<rgba 1 0 1 1>}
  n214 = ScalarParameter() {
   defaultValue=1.5
   desc=Linear interpolates the texture mask from the low range to the high range.
   group=00_Base
   parameterName=Base_BaseColorNoise_B_LowRange
   sortPriority=-430
  }
  n215 = ScalarParameter() {
   defaultValue=-0.5
   desc=Linear interpolates the texture mask from the low range to the high range.
   group=00_Base
   parameterName=Base_BaseColorNoise_B_HighRange
   sortPriority=-420
  }
  n216 = ScalarParameter() {
   defaultValue=10100
   desc=Scale of the textures in the level. A lower value decreases the size.
   group=00_Base
   parameterName=Base_BaseColorNoise_B_TextureScale
   sortPriority=-450
  }
  n217 = StaticComponentMaskParameter(None=MaterialFunctionCall(Export Float 4=@n210, TextureObject=@n211, TextureSize=@n216) {materialFunction=&WorldAlignedTexture}.Z Texture) {
   defaultG=true
   desc=Define the texture channel to use for BaseColorNoise_B.
   group=00_Base
   parameterName=Base_BaseColorNoise_B_MaskChannel
   sortPriority=-440
  }
  NamedRerouteDeclaration(None=Saturate(None=LinearInterpolate(A=@n214, B=@n215, Alpha=@n217))) {name=Noise_B, nodeColor=<rgba 1 0 1 1>}
  NamedRerouteDeclaration(None=ComponentMask(None=@n157.Z Diffuse Texture) {b=true, g=true, r=true}) {name=Base_RawBaseColor, nodeColor=<rgba 1 0 1 1>}
  n218 = VectorParameter() {
   desc="Divide the adjusted BaseColor by this color to get the slope color.\r\nA vaule of 1,1,1 will not alter the colors.\r\nAlpha value is not used."
   group=00_Base
   parameterName=Base_BaseColorSlope_Tint
   sortPriority=-197
   channelNames {a=Not Used, b=Blue Value, g=Green Value, r=Red Value}
   defaultValue=<rgba 1 1 1 1>
  }
  n219 = NamedRerouteUsage()
  n220 = Multiply(A=ComponentMask(None=@n218) {b=true, g=true, r=true}, B=@n219)
  n221 = Constant()
  n222 = ComponentMask(None=VertexNormalWS()) {b=true}
  n223 = ScalarParameter() {
   defaultValue=6
   desc="Take the vertex normal of the landscape, mask the blue channel, then use this value as the exponent of that result.\r\nLarger values will \"pull the mask in\", isolating it to areas with a large degree of slope."
   group=00_Base
   parameterName=Base_BaseColorSlope_MaskPower
   sortPriority=-199
  }
  n224 = ScalarParameter() {
   defaultValue=1.5
   desc="Multiply the exponent result by this value, then clamp it 0-1."
   group=00_Base
   parameterName=Base_BaseColorSlope_MaskMultiply
   sortPriority=-198
  }
  n225 = Saturate(None=Multiply(A=Power(Base=@n222, Exp=@n223), B=@n224))
  n226 = ScalarParameter() {
   defaultValue=1
   desc="BaseColorSlope adjusts the BaseColor using the degree of slope of the landscape mesh.\r\nA value of zero will ignore BaseColorSlope.\r\nThis value is clamped 0-1."
   group=00_Base
   parameterName=Base_BaseColorSlope_Amount
   sortPriority=-200
  }
  n227 = Saturate(None=@n226)
  NamedRerouteDeclaration(None=LinearInterpolate(A=Saturate(None=@n220), B=@n219, Alpha=LinearInterpolate(A=@n221, B=@n225, Alpha=@n227))) {name=Base_BaseColorFinal, nodeColor=<rgba 1 0 1 1>}
  n228 = VectorParameter() {
   desc="A vaule of 1,1,1 will not alter the colors.\r\nAlpha value is not used."
   group=00_Base
   parameterName=Base_BaseColorNoise_Tint_Color1
   sortPriority=-300
   defaultValue=<rgba 1 1 1 1>
  }
  n229 = ComponentMask(None=@n228.RGB) {b=true, g=true, r=true}
  n230 = VectorParameter() {
   desc="A vaule of 1,1,1 will not alter the colors.\r\nAlpha value is not used."
   group=00_Base
   parameterName=Base_BaseColorNoise_Tint_Color2
   sortPriority=-290
   defaultValue=<rgba 1 1 1 1>
  }
  n231 = ComponentMask(None=@n230.RGB) {b=true, g=true, r=true}
  n232 = Multiply(A=NamedRerouteUsage(), B=NamedRerouteUsage())
  n233 = Saturate(None=Multiply(A=NamedRerouteUsage(), B=LinearInterpolate(A=@n229, B=@n231, Alpha=@n232)))
  n234 = VectorParameter() {group=00_Base, parameterName=Base_BaseColorNoise_Tint_Color3, sortPriority=-280, defaultValue=<rgba 1 1 1 1>}
  n235 = Multiply(A=@n233, B=ComponentMask(None=@n234.RGB) {b=true, g=true, r=true})
  n236 = LinearInterpolate(A=NamedRerouteUsage(), B=Constant(), Alpha=NamedRerouteUsage())
  n237 = VectorParameter() {
   desc="Multiply the BaseColor_Texture by a color.\r\nAlpha value is not used."
   group=00_Base
   parameterName=Base_BaseColor_Tint
   sortPriority=-499
   defaultValue=<rgba 1 1 1 1>
  }
  n238 = Multiply(A=NamedRerouteUsage(), B=ComponentMask(None=@n237.RGB) {b=true, g=true, r=true})
  NamedRerouteDeclaration(None=StaticSwitchParameter(True=LinearInterpolate(A=@n233, B=@n235, Alpha=@n236), False=Saturate(None=@n238)) {desc=Uses a texture for color variation., group=00_Base, parameterName=Base_Use_BaseColorNoise, sortPriority=-500}) {name=Base_AdjustedBasecolor, nodeColor=<rgba 1 0 1 1>}
  n239 = ScalarParameter() {
   defaultValue=2
   desc=Linear interpolates the texture mask from the low range to the high range.
   group=00_Base
   parameterName=Base_BaseColorNoise_A_LowRange
   sortPriority=-470
  }
  n240 = ScalarParameter() {
   defaultValue=-0.5
   desc=Linear interpolates the texture mask from the low range to the high range.
   group=00_Base
   parameterName=Base_BaseColorNoise_A_HighRange
   sortPriority=-460
  }
  n241 = ScalarParameter() {
   defaultValue=2500
   desc=Scale of the textures in the level. A lower value decreases the size.
   group=00_Base
   parameterName=Base_BaseColorNoise_A_TextureScale
   sortPriority=-490
  }
  n242 = StaticComponentMaskParameter(None=MaterialFunctionCall(Export Float 4=@n210, TextureObject=@n211, TextureSize=@n241) {materialFunction=&WorldAlignedTexture}.Z Texture) {
   defaultG=true
   desc=Define the texture channel to use for BaseColorNoise_A.
   group=00_Base
   parameterName=Base_BaseColorNoise_A_MaskChannel
   sortPriority=-480
  }
  NamedRerouteDeclaration(None=Saturate(None=LinearInterpolate(A=@n239, B=@n240, Alpha=@n242))) {name=Noise_A, nodeColor=<rgba 1 0 1 1>}
  n243 = NamedRerouteUsage()
  n244 = VectorParameter() {
   desc="Divide all the underlaying layer's BaseColor by this color.\r\nA vaule of 1,1,1 will not alter the colors.\r\nAlpha value is not used."
   group=07_Layer7
   parameterName=Layer7_BaseColor_Tint1
   sortPriority=10
   channelNames {a=Not Used, b=Blue Value, g=Green Value, r=Red Value}
   defaultValue=<rgba 0.5 0.5 0.5 1>
  }
  n245 = VectorParameter() {
   desc="Multiply the divided BaseColor by this color.\r\nAlpha value is not used."
   group=07_Layer7
   parameterName=Layer7_BaseColor_Tint2
   sortPriority=20
   channelNames {a=Not Used, b=Blue Value, g=Green Value, r=Red Value}
   defaultValue=<rgba 1 1 1 1>
  }
  n246 = Multiply(A=Saturate(None=Multiply(A=@n243, B=@n244)), B=@n245)
  NamedRerouteDeclaration(None=LinearInterpolate(A=Saturate(None=@n246), B=NamedRerouteUsage(), Alpha=NamedRerouteUsage())) {name=Layer7_Final_Basecolor, nodeColor=<rgba 0.5 0 1 1>}
  n247 = NamedRerouteUsage()
  n248 = VertexNormalWS()
  n249 = ScalarParameter() {
   defaultValue=0.5
   desc="Fades out the underlaying layers normal maps creating a \"flat\" look."
   group=07_Layer7
   parameterName=Layer7_FlattenNormals
   sortPriority=60
  }
  NamedRerouteDeclaration(None=LinearInterpolate(A=LinearInterpolate(A=@n247, B=@n248, Alpha=@n249), B=NamedRerouteUsage(), Alpha=NamedRerouteUsage())) {name=Layer7_Final_Normal, nodeColor=<rgba 0.5 0 1 1>}
  n250 = ScalarParameter() {
   defaultValue=0.01
   desc="Set the minimum && maximum to the same value if a singular specular value is desired."
   group=07_Layer7
   parameterName=Layer7_Specular_Minimum
   sortPriority=30
  }
  n251 = ScalarParameter() {
   defaultValue=0.1
   desc="Set the minimum && maximum to the same value if a singular specular value is desired."
   group=07_Layer7
   parameterName=Layer7_Specular_Maximum
   sortPriority=31
  }
  n252 = NamedRerouteUsage()
  NamedRerouteDeclaration(None=LinearInterpolate(A=LinearInterpolate(A=@n250, B=@n251, Alpha=@n252), B=NamedRerouteUsage(), Alpha=NamedRerouteUsage())) {name=Layer7_Final_Spec, nodeColor=<rgba 0.5 0 1 1>}
  n253 = SetMaterialAttributes!(
   MaterialAttributes=@n126.MaterialAttributes
   Base Color=NamedRerouteUsage()
   Normal=NamedRerouteUsage()
   Specular=NamedRerouteUsage()
   Roughness=NamedRerouteUsage()
   Metallic=NamedRerouteUsage()
  ) {attributeSetTypes [
    #2
    #3
    #4
    #5
    #6
   ]}
  NamedRerouteDeclaration(None=LandscapeLayerSwitch(LayerUsed=@n253, LayerNotUsed=@n125) {parameterName=Layer7}) {name=BlendedLayers, nodeColor=<rgba 0.5 0 1 1>}
  n254 = ScalarParameter() {
   defaultValue=0.3
   desc="Set the minimum && maximum to the same value if a singular roughness value is desired."
   group=07_Layer7
   parameterName=Layer7_Roughness_Minimum
   sortPriority=44
  }
  n255 = ScalarParameter() {
   defaultValue=0.1
   desc="Set the minimum && maximum to the same value if a singular roughness value is desired."
   group=07_Layer7
   parameterName=Layer7_Roughness_Maximum
   sortPriority=45
  }
  n256 = ScalarParameter() {
   defaultValue=2
   desc=Set the minimum value of the heightmap mask.
   group=07_Layer7
   parameterName=Layer7_HeightmapAsRoughness_Minimum
   sortPriority=42
  }
  n257 = ScalarParameter() {
   defaultValue=-2
   desc=Set the maximum value of the heightmap mask.
   group=07_Layer7
   parameterName=Layer7_HeightmapAsRoughness_Maximum
   sortPriority=43
  }
  n258 = LinearInterpolate(A=@n256, B=@n257, Alpha=NamedRerouteUsage())
  n259 = ScalarParameter() {
   defaultValue=1
   desc="Uses the heightmap from the layer that is painted over. This can allow for interesting roughness variation.\r\nA value of zero will use the default roughness."
   group=07_Layer7
   parameterName=Layer7_UseHeightmapAsRoughness
   sortPriority=41
  }
  n260 = LinearInterpolate(A=@n254, B=@n255, Alpha=LinearInterpolate(A=NamedRerouteUsage(), B=Saturate(None=@n258), Alpha=Saturate(None=@n259)))
  NamedRerouteDeclaration(None=LinearInterpolate(A=Saturate(None=@n260), B=NamedRerouteUsage(), Alpha=NamedRerouteUsage())) {name=Layer7_Final_Rough, nodeColor=<rgba 0.5 0 1 1>}
  n261 = ScalarParameter() {
   desc="Set the minimum && maximum to the same value if a singular metallic value is desired."
   group=07_Layer7
   parameterName=Layer7_Metallic_Minimum
   sortPriority=50
  }
  n262 = ScalarParameter() {
   defaultValue=1
   desc="Set the minimum && maximum to the same value if a singular metallic value is desired."
   group=07_Layer7
   parameterName=Layer7_Metallic_Maximum
   sortPriority=51
  }
  n263 = NamedRerouteUsage()
  NamedRerouteDeclaration(None=LinearInterpolate(A=LinearInterpolate(A=@n261, B=@n262, Alpha=@n263), B=NamedRerouteUsage(), Alpha=NamedRerouteUsage())) {name=Layer7_Final_Metal, nodeColor=<rgba 0.5 0 1 1>}
  NamedRerouteDeclaration(None=@n111.Alpha) {name=Layer1Mask, nodeColor=<rgba 0.5 0 1 1>}
  NamedRerouteDeclaration(None=@n113.Alpha) {name=Layer2Mask, nodeColor=<rgba 0.5 0 1 1>}
  NamedRerouteDeclaration(None=@n115.Alpha) {name=Layer3Mask, nodeColor=<rgba 0.5 0 1 1>}
  NamedRerouteDeclaration(None=@n117.Alpha) {name=Layer4Mask, nodeColor=<rgba 0.5 0 1 1>}
  NamedRerouteDeclaration(None=@n119.Alpha) {name=Layer5Mask, nodeColor=<rgba 0.5 0 1 1>}
  n264 = ScalarParameter() {
   defaultValue=1
   desc=A higher value will sharpen the blending to other layers.
   group=01_Layer1
   parameterName=Layer1_BaseColor_Heightmap_LayerBlendContrast
   sortPriority=140
  }
  NamedRerouteDeclaration(None=Abs(None=@n264)) {name=Layer1_HeightBlendContrast, nodeColor=<rgba 0.5 0 1 1>}
  NamedRerouteDeclaration(None=@n121.Alpha) {name=Layer6Mask, nodeColor=<rgba 0.5 0 1 1>}
  NamedRerouteDeclaration(None=StaticSwitch(True=@n123.Alpha, False=Constant() {r=1}, Value=@n124)) {name=CliffMask, nodeColor=<rgba 0.5 0 1 1>}
  NamedRerouteDeclaration(None=@n121.Resulting Height Map) {name=Layer6HeightResult, nodeColor=<rgba 0.5 0 1 1>}
  NamedRerouteDeclaration(None=StaticSwitch(True=@n123.Resulting Height Map, False=NamedRerouteUsage(), Value=@n124)) {name=HeightMapWithCliff, nodeColor=<rgba 0.5 0 1 1>}
  n265 = ScalarParameter() {
   defaultValue=1
   desc=A higher value will sharpen the blending to other layers.
   group=02_Layer2
   parameterName=Layer2_BaseColor_Heightmap_LayerBlendContrast
   sortPriority=140
  }
  NamedRerouteDeclaration(None=Abs(None=@n265)) {name=Layer2_HeightBlendContrast, nodeColor=<rgba 0.5 0 1 1>}
  n266 = ScalarParameter() {
   defaultValue=1
   desc=A higher value will sharpen the blending to other layers.
   group=03_Layer3
   parameterName=Layer3_BaseColor_Heightmap_LayerBlendContrast
   sortPriority=140
  }
  NamedRerouteDeclaration(None=Abs(None=@n266)) {name=Layer3_HeightBlendContrast, nodeColor=<rgba 0.5 0 1 1>}
 ]
}
